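; Projectile Blueprint for the Grappling Hoot item, which creates the grapple rope actor.
; SpawnRope spawns B_Athena_GrapplingHoot_Rope at an actor location, passing the weapon actor and player pawn, then stores it in the Rope variable.
; It links the rope to the weapon and this projectile through SetWeaponandProjectile and returns it.
; BeginPlay and construction only call the parent, while overlap and tick events are empty.
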(event EventBeginPlay
  (|CallParentFunction))

(event Collision|EventActorBeginOverlap
  (params (OtherActor :type Actor%20Object%20Reference)))

(event EventTick
  (params (DeltaSeconds :type Float%20%28single-precision%29)))

(fn SpawnRope
  (params (PlayerPawn :type Fort%20Player%20Pawn%20Object%20Reference))
  ;[1 1 1] Spawn The Rope
  (bind _returnvalue (Game|SpawnActorfromClass "/Game/Athena/Items/Gameplay/GrapplingHoot/B_Athena_GrapplingHoot_Rope.B_Athena_GrapplingHoot_Rope_C" (Math|Transform|MakeTransform (Transformation|GetActorLocation)) "Undefined" "MultiplyWithRoot" (Variables|Default|GetWeapon%20Actor) PlayerPawn 0))
  (bind _output_get (Variables|Default|SetRope _returnvalue))
  (Class|FortHookGunRope|SetWeaponandProjectile _output_get (Variables|Default|GetWeapon%20Actor) self)
  (return _output_get))

(fn ConstructionScript
  (|CallParentFunction))
